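# Post-process style Anim Blueprint for the Aline female head that adds simulated hair motion to the incoming pose.
# Six AnimDynamics nodes run in series on the dyn_hair bones at partial alphas, mixing angular-spring, linear-spring and bias-driven setups, with LOD thresholds on some.
# On initialization it caches the pawn head mesh's Hair_Hat morph value.
# That value fades the dyn_hair_05 and dyn_hair_05_a simulations to zero as it reaches one.
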AnimBlueprint /BRCosmetics/Characters/Player/Female/Medium/Heads/F_MED_HIS_Aline_Head_01/Meshes/F_MED_HIS_Aline_Head_01_Dyn_AnimBP
target_skeleton=/Game/Characters/Player/Common/Fortnite_Base_Head/Fortnite_Base_Head_Export_Skeleton
graphs [
 AnimGraph {nodes [
   AnimGraphNode_SubInput_0 = Input Pose()
   n1 = Anim Dynamics - Bone: dyn_hair_01(ComponentPose=Local To Component(LocalPose=@AnimGraphNode_SubInput_0), Alpha=0.5) {
    alpha=0.9
    angularSpringConstant=1500
    bAngularSpring=true
    bGravityOverrideInSimSpace=true
    gravityScale=0.65
    numSolverIterationsPostUpdate=4
    numSolverIterationsPreUpdate=16
    simulationSpace=Actor
    alphaBoolBlend {blendOption=Linear}
    boundBone {boneName=dyn_hair_01}
    physicsBodyDefinitions [{
      sphereCollisionRadius=0
      boundBone {boneName=dyn_hair_01}
      boxExtents=<15 15 2.5>
      constraintSetup {
       angularTargetAxis=AxisZ
       twistAxis=AxisZ
       angularLimitsMax=<15 15 15>
       angularLimitsMin=<-15 -15 -15>
       angularTarget=<0 0 1>
       linearAxesMax=<0 3 0>
       linearAxesMin=<0 -3 0>
      }
      localJointOffset=<0 0 -10>
     }]
   }
   n2 = Anim Dynamics - Bone: dyn_hair_02(ComponentPose=@n1, Alpha=0.3) {
    alpha=0.9
    angularSpringConstant=1500
    bAngularSpring=true
    bGravityOverrideInSimSpace=true
    gravityScale=0.65
    numSolverIterationsPostUpdate=4
    numSolverIterationsPreUpdate=16
    simulationSpace=Actor
    alphaBoolBlend {blendOption=Linear}
    boundBone {boneName=dyn_hair_02}
    physicsBodyDefinitions [{
      sphereCollisionRadius=0
      boundBone {boneName=dyn_hair_02}
      boxExtents=<15 15 2.5>
      constraintSetup {
       angularTargetAxis=AxisZ
       twistAxis=AxisZ
       angularLimitsMax=<15 15 15>
       angularLimitsMin=<-15 -15 -15>
       angularTarget=<0 0 1>
       linearAxesMax=<0 3 0>
       linearAxesMin=<0 -3 0>
      }
      localJointOffset=<0 0 -10>
     }]
   }
   n3 = Anim Dynamics - Bone: dyn_hair_03(ComponentPose=@n2, Alpha=0.23, ExternalForce="0,0,0") {
    angularSpringConstant=60
    bAngularSpring=true
    bGravityOverrideInSimSpace=true
    bLinearSpring=true
    gravityScale=0
    lODThreshold=1
    linearSpringConstant=60
    simulationSpace=Actor
    alphaBoolBlend {blendOption=Linear}
    boundBone {boneName=dyn_hair_03}
    physicsBodyDefinitions [{
      sphereCollisionRadius=0
      boundBone {boneName=dyn_hair_03}
      boxExtents=<30 30 30>
      constraintSetup {
       twistAxis=AxisZ
       angularLimitsMax=<0 0 0>
       angularLimitsMin=<0 0 0>
       angularTarget=<0 0 0>
       linearAxesMax=<0 0.5 2.5>
       linearAxesMin=<0 -0.5 -0.5>
      }
      localJointOffset=<0 0 0>
     }]
    relativeSpaceBone {boneName=dyn_hair_01_b}
   }
   n4 = Anim Dynamics - Bone: dyn_hair_04(ComponentPose=@n3, Alpha=0.2, ExternalForce="0,0,0") {
    angularSpringConstant=65
    bAngularSpring=true
    bGravityOverrideInSimSpace=true
    bLinearSpring=true
    gravityScale=0
    lODThreshold=1
    linearSpringConstant=65
    simulationSpace=Actor
    alphaBoolBlend {blendOption=Linear}
    boundBone {boneName=dyn_hair_04}
    physicsBodyDefinitions [{
      sphereCollisionRadius=0
      boundBone {boneName=dyn_hair_04}
      boxExtents=<30 30 30>
      constraintSetup {
       twistAxis=AxisZ
       angularLimitsMax=<0 0 0>
       angularLimitsMin=<0 0 0>
       angularTarget=<0 0 0>
       linearAxesMax=<0 0.5 5>
       linearAxesMin=<0 -0.5 -0.5>
      }
      localJointOffset=<0 0 0>
     }]
    relativeSpaceBone {boneName=dyn_hair_01_a}
   }
   n5 = Get Hair_Hat() -> (Hair_Hat:double)
   n6 = Anim Dynamics - Bone: dyn_hair_05(ComponentPose=@n4, Alpha=MapRangeClamped(Value=@n5, InRangeA=0, InRangeB=1, OutRangeA=0.05, OutRangeB=0), ExternalForce="0,0,0") {
    angularSpringConstant=20
    bAngularSpring=true
    bGravityOverrideInSimSpace=true
    bLinearSpring=true
    gravityScale=0
    lODThreshold=1
    linearSpringConstant=20
    simulationSpace=Actor
    alphaBoolBlend {blendOption=Linear}
    boundBone {boneName=dyn_hair_05}
    physicsBodyDefinitions [{
      sphereCollisionRadius=0
      boundBone {boneName=dyn_hair_05}
      boxExtents=<30 30 30>
      constraintSetup {
       twistAxis=AxisZ
       angularLimitsMax=<3 7.5 0>
       angularLimitsMin=<-3 -7.5 0>
       angularTarget=<0 0 0>
       linearAxesMax=<1.5 6 5>
       linearAxesMin=<-1.5 -6 -0.5>
      }
      localJointOffset=<0 0 0>
     }]
    relativeSpaceBone {boneName=dyn_hair_01}
   }
   n7 = Anim Dynamics - Bone: dyn_hair_05_a(ComponentPose=@n6, Alpha=MapRangeClamped(Value=@n5, InRangeA=0, InRangeB=1, OutRangeA=0.06, OutRangeB=0), ExternalForce="0,0,0") {
    angularBiasOverride=1
    bGravityOverrideInSimSpace=true
    bOverrideAngularBias=true
    gravityScale=0.1
    alphaBoolBlend {blendOption=Linear}
    boundBone {boneName=dyn_hair_05_a}
    physicsBodyDefinitions [{
      sphereCollisionRadius=0
      boundBone {boneName=dyn_hair_05_a}
      boxExtents=<5 5 5>
      constraintSetup {
       angularTargetAxis=AxisY
       twistAxis=AxisY
       angularLimitsMax=<10 0 10>
       angularLimitsMin=<-10 0 -10>
       angularTarget=<0 0 0>
       linearAxesMax=<0 0 3>
      }
      localJointOffset=<0 0 -2>
     }]
   }
   Output Pose(Result=Component To Local(ComponentPose=@n7)) {layerGroup=None}
  ]}
 EventGraph {nodes [
   Event BlueprintInitializeAnimation()
   > n1 = Cast To FortPlayerPawn(Object=GetOwner(self=GetPartSkeletalMeshComponent()))
   > Set Hair_Hat(Hair_Hat:double=GetMorphTarget!(self:Skeletal Mesh Component Object Reference=GetSkeletalMeshForPartType(self=@n1, PartType=Head), MorphTargetName:Name=Hair_Hat) -> (ReturnValue:float)) -> (Output_Get:double)
   PrintString!(execute, InString="To String (Float)"(InDouble=0), bPrintToScreen, bPrintToLog, TextColor, Duration, Key) -> (then)
  ]}
]
variables [Hair_Hat {type=number}]
